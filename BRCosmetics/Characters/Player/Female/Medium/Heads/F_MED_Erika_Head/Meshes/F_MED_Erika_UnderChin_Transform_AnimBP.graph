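# Post-process style AnimBlueprint for the Erika head that restricts head rotation when a specific helmet is the equipped Face part.
# Each update scales and negates the head-to-neck_02 delta rotation, then applies it additively to the head bone when a pawn owner exists.
# This applies only outside emotes and the front end.
# A jawOpen curve-driven under-chin offset is cached as UnderChin_fix_pose, but the output pose never uses it.
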AnimBlueprint /BRCosmetics/Characters/Player/Female/Medium/Heads/F_MED_Erika_Head/Meshes/F_MED_Erika_UnderChin_Transform_AnimBP
target_skeleton=/Game/Characters/Player/Common/Fortnite_Base_Head/Fortnite_Base_3L_Head_Skeleton
graphs [
 AnimGraph {nodes [
   n1 = "Transform (Modify) Bone - Bone: head"(
    ComponentPose=Local To Component(LocalPose=Use cached pose 'Base_pose'())
    Alpha=Get enableHeadRotLimitAlpha() -> (enableHeadRotLimitAlpha:double)
    Translation="0,0,0"
    Rotation=Get Head_Rot() -> (Head_Rot:Rotator)
    Scale="1,1,1"
   ) {rotationMode=BMM_Additive, rotationSpace=BCS_BoneSpace, alphaBoolBlend {blendOption=Linear}, boneToModify {boneName=head}}
   n2 = Component To Local(ComponentPose=@n1)
   n3 = Use cached pose 'Base_pose'()
   n4 = Get Head_Limit() -> (Head_Limit:bool)
   Output Pose(Result=Blend Poses by bool(
     BlendPose_0=Blend Poses by bool(BlendPose_0=@n2, BlendPose_1=@n3, bActiveValue=@n4) {blendType=Linear}
     BlendPose_1=Use cached pose 'Base_pose'()
     bActiveValue=Get Helmet_On() -> (Helmet_On:bool)
    ) {blendType=Linear}) {layerGroup=None}
   Save cached pose 'Base_pose'(Pose=Input Pose()) {cacheName=Base_pose}
   n5 = Get JawOpen() -> (JawOpen:double)
   n6 = "Transform (Modify) Bone - Bone: FACIAL_C_UnderChin"(ComponentPose=Local To Component(LocalPose=Use cached pose 'Base_pose'()), Alpha=@n5, Translation="0,-3,-5", Rotation="0,0,0", Scale="1,1,1") {
    translationMode=BMM_Additive
    translationSpace=BCS_BoneSpace
    alphaBoolBlend {blendOption=Linear}
    boneToModify {boneName=FACIAL_C_UnderChin}
    translation {y=-3, z=-5}
   }
   n7 = "Transform (Modify) Bone - Bone: FACIAL_R_UnderChin"(ComponentPose=@n6, Alpha=@n5, Translation="0,-3,-5", Rotation="0,0,0", Scale="1,1,1") {
    translationMode=BMM_Additive
    translationSpace=BCS_BoneSpace
    alphaBoolBlend {blendOption=Linear}
    boneToModify {boneName=FACIAL_R_UnderChin}
    translation {y=-3, z=-5}
   }
   n8 = "Transform (Modify) Bone - Bone: FACIAL_L_UnderChin"(ComponentPose=@n7, Alpha=@n5, Translation="0,-3,-5", Rotation="0,0,0", Scale="1,1,1") {
    translationMode=BMM_Additive
    translationSpace=BCS_BoneSpace
    alphaBoolBlend {blendOption=Linear}
    boneToModify {boneName=FACIAL_L_UnderChin}
    translation {y=-3, z=-5}
   }
   Save cached pose 'UnderChin_fix_pose'(Pose=Component To Local(ComponentPose=@n8)) {cacheName=UnderChin_fix_pose}
  ]}
 EventGraph {nodes [
   n1 = Event OnCharacterCustomizationLoaded()
   n2 = GetSkinnedAsset!(self:Skinned Mesh Component Object Reference=Get Array Item with EFort Custom Part Type(Array=@n1.CustomMeshParts, PartType=Face)) -> (ReturnValue:Skinned Asset Object Reference)
   > Set Helmet_On(Helmet_On:bool="Equal Exactly (String)"(A=GetDisplayName(Object=@n2), B=Get isHelmet() -> (isHelmet:String))) -> (Output_Get:bool)
   Event BlueprintUpdateAnimation()
   n3 = GetPartSkeletalMeshComponent()
   > Set JawOpen(JawOpen:double=GetCurveValue(self=GetAnimInstance(self=@n3), CurveName=CTRL_expressions_jawOpen)) -> (Output_Get:double)
   n4 = GetOwnerSkeletalMeshComponent()
   n5 = BreakRotator(InRot=BreakTransform(InTransform=GetDeltaTransformFromRefPose(self=@n4, BoneName=head, BaseName=neck_02)).Rotation)
   n6 = Get Y_percentage() -> (Y_percentage:double)
   n7 = Get X_percentage() -> (X_percentage:double)
   n8 = Get Z_percentage() -> (Z_percentage:double)
   > Set Head_Rot(Head_Rot:Rotator=MakeRotator(
     Roll=float * float!(A=@n5, B=-1, C:double=@n7) -> (ReturnValue)
     Pitch=float * float(A=@n5.Pitch, B=-1, C=@n6)
     Yaw=float * float(A=@n5.Yaw, B=-1, C=@n8)
    )) -> (Output_Get:Rotator)
   > n9 = Is Valid(InputObject=TryGetPawnOwner())
   Set enableHeadRotLimitAlpha(execute=@n9.Is Not Valid, enableHeadRotLimitAlpha:double=0) -> (Output_Get:double)
   Set enableHeadRotLimitAlpha(execute=@n9.Is Valid, enableHeadRotLimitAlpha=1)
   n10 = Get bIsInFrontEnd()
   n11 = Get bIsPlayingEmote()
   > Set Head_Limit(Head_Limit:bool=AND Boolean(A=NOT Boolean(A=@n11), B=NOT Boolean(A=@n10))) -> (Output_Get:bool)
  ]}
]
variables [
 enableHeadRotLimitAlpha {type=number}
 Head_Limit {type=boolean}
 Head_Rot {type=Rotator}
 Helmet_On {type=boolean}
 isHelmet {type=string, default=F_MED_CoreStreet_Helmet_FaceAcc}
 JawOpen {type=number}
 Out Names {type="array<string>", default []}
 X_percentage {type=number, default=0.8}
 Y_percentage {type=number, default=0.8}
 Z_percentage {type=number, default=0.9}
]
